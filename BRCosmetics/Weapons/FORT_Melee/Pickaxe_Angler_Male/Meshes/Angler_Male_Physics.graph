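# Physics asset for the Angler Male pickaxe mesh, defining bodies and joints for five dynamic fins, apparently for secondary motion.
# Each fin has an offset capsule body jointed to its own tiny MSTR body with narrow, asymmetric swing limits and a tight twist limit.
# The five MSTR bodies each joint to one shared base body with free swing and twist, leaving their stored angle limits unused.
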PhysicsAsset /BRCosmetics/Weapons/FORT_Melee/Pickaxe_Angler_Male/Meshes/Angler_Male_Physics
bodies {
 dyn_fins_01 [{shapeType=Capsule, length=7.9354, radius=3.5, center=<5 0 0>, rotation=<rot 90 0 0>}]
 dyn_fins_01_MSTR [{shapeType=Capsule, length=0.5, radius=0.5}]
 dyn_fins_02 [{shapeType=Capsule, length=9.0823, radius=3.5, center=<8 0 0>, rotation=<rot 90 0 0>}]
 dyn_fins_02_MSTR [{shapeType=Capsule, length=0.5, radius=0.5}]
 dyn_fins_03 [{shapeType=Capsule, length=9.2412, radius=3.5, center=<10 0 0>, rotation=<rot 90 0 0>}]
 dyn_fins_03_MSTR [{shapeType=Capsule, length=0.5, radius=0.5}]
 dyn_fins_04 [{shapeType=Capsule, length=6.8033, radius=3.5, center=<6 0 0>, rotation=<rot 90 0 0>}]
 dyn_fins_04_MSTR [{shapeType=Capsule, length=0.5, radius=0.5}]
 dyn_fins_05 [{shapeType=Capsule, length=4.4297, radius=3.5, center=<3 0 0>, rotation=<rot 90 0 0>}]
 dyn_fins_05_MSTR [{shapeType=Capsule, length=0.5, radius=0.5}]
 dyn_fins_base [{shapeType=Capsule, length=0.5, radius=0.5}]
}
constraints [
 {bone1Name=dyn_fins_01, bone2Name=dyn_fins_01_MSTR, swing1LimitDegrees=12, swing2LimitDegrees=30, twistLimitDegrees=5}
 {bone1Name=dyn_fins_02, bone2Name=dyn_fins_02_MSTR, swing1LimitDegrees=12, swing2LimitDegrees=30, twistLimitDegrees=5}
 {bone1Name=dyn_fins_03, bone2Name=dyn_fins_03_MSTR, swing1LimitDegrees=12, swing2LimitDegrees=30, twistLimitDegrees=5}
 {bone1Name=dyn_fins_04, bone2Name=dyn_fins_04_MSTR, swing1LimitDegrees=12, swing2LimitDegrees=30, twistLimitDegrees=5}
 {bone1Name=dyn_fins_05, bone2Name=dyn_fins_05_MSTR, swing1LimitDegrees=12, swing2LimitDegrees=30, twistLimitDegrees=5}
 {
  bone1Name=dyn_fins_01_MSTR
  bone2Name=dyn_fins_base
  swing1LimitDegrees=45
  swing1Motion=Free
  swing2LimitDegrees=45
  swing2Motion=Free
  twistLimitDegrees=45
  twistMotion=Free
 }
 {
  bone1Name=dyn_fins_02_MSTR
  bone2Name=dyn_fins_base
  swing1LimitDegrees=45
  swing1Motion=Free
  swing2LimitDegrees=45
  swing2Motion=Free
  twistLimitDegrees=45
  twistMotion=Free
 }
 {
  bone1Name=dyn_fins_03_MSTR
  bone2Name=dyn_fins_base
  swing1LimitDegrees=45
  swing1Motion=Free
  swing2LimitDegrees=45
  swing2Motion=Free
  twistLimitDegrees=45
  twistMotion=Free
 }
 {
  bone1Name=dyn_fins_04_MSTR
  bone2Name=dyn_fins_base
  swing1LimitDegrees=45
  swing1Motion=Free
  swing2LimitDegrees=45
  swing2Motion=Free
  twistLimitDegrees=45
  twistMotion=Free
 }
 {
  bone1Name=dyn_fins_05_MSTR
  bone2Name=dyn_fins_base
  swing1LimitDegrees=45
  swing1Motion=Free
  swing2LimitDegrees=45
  swing2Motion=Free
  twistLimitDegrees=45
  twistMotion=Free
 }
]
