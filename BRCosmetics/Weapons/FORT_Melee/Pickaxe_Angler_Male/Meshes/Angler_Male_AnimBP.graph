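# Animation Blueprint for the Angler male pickaxe skeleton, adding physics-driven secondary motion to the melee weapon.
# The input pose runs through a RigidBody node simulating in b_AR_Root space with overridden gravity and scaled component motion, then goes to output.
# A time-based sine noise function and additive fin-bone rotation chain are also present, but the update event is unwired and the chain never reaches the output pose.
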AnimBlueprint /BRCosmetics/Weapons/FORT_Melee/Pickaxe_Angler_Male/Meshes/Angler_Male_AnimBP
target_skeleton=&/Angler_Male_Skeleton
graphs [
 AnimGraph {nodes [
   n1 = Get RotateNoiseA() -> (RotateNoiseA:Rotator)
   n2 = "Transform (Modify) Bone - Bone: dyn_fins_01_MSTR"(Alpha=1, Translation="0,0,0", Rotation=@n1, Scale="1,1,1") {rotationMode=BMM_Additive, rotationSpace=BCS_BoneSpace, alphaBoolBlend {blendOption=Linear}, boneToModify {boneName=dyn_fins_01_MSTR}}
   n3 = Get RotateNoiseB() -> (RotateNoiseB:Rotator)
   n4 = "Transform (Modify) Bone - Bone: dyn_fins_02_MSTR"(ComponentPose=@n2, Alpha=1, Translation="0,0,0", Rotation=@n3, Scale="1,1,1") {rotationMode=BMM_Additive, rotationSpace=BCS_BoneSpace, alphaBoolBlend {blendOption=Linear}, boneToModify {boneName=dyn_fins_02_MSTR}}
   n5 = "Transform (Modify) Bone - Bone: dyn_fins_03_MSTR"(ComponentPose=@n4, Alpha=1, Translation="0,0,0", Rotation=Get RotateNoiseC() -> (RotateNoiseC:Rotator), Scale="1,1,1") {rotationMode=BMM_Additive, rotationSpace=BCS_BoneSpace, alphaBoolBlend {blendOption=Linear}, boneToModify {boneName=dyn_fins_03_MSTR}}
   n6 = "Transform (Modify) Bone - Bone: dyn_fins_04_MSTR"(ComponentPose=@n5, Alpha=1, Translation="0,0,0", Rotation=@n3, Scale="1,1,1") {rotationMode=BMM_Additive, rotationSpace=BCS_BoneSpace, alphaBoolBlend {blendOption=Linear}, boneToModify {boneName=dyn_fins_04_MSTR}}
   "Transform (Modify) Bone - Bone: dyn_fins_05_MSTR"(ComponentPose=@n6, Alpha=1, Translation="0,0,0", Rotation=@n1, Scale="1,1,1") {rotationMode=BMM_Additive, rotationSpace=BCS_BoneSpace, alphaBoolBlend {blendOption=Linear}, boneToModify {boneName=dyn_fins_05_MSTR}}
   AnimGraphNode_SubInput_0 = Input Pose()
   n7 = RigidBody(ComponentPose=Local To Component(LocalPose=@AnimGraphNode_SubInput_0)) {
    bClampLinearTranslationLimitToRefPose=true
    bOverrideWorldGravity=true
    overridePhysicsAsset=&/Angler_Male_Physics
    simulationSpace=BaseBoneSpace
    alphaBoolBlend {blendOption=Linear}
    baseBoneRef {boneName=b_AR_Root}
    componentAppliedLinearAccClamp=<100 50 50>
    componentLinearAccScale=<0.2 0.1 0.1>
    componentLinearVelScale=<0.5 0.5 0.5>
    overrideWorldGravity {z=-20}
   }
   Output Pose(Result=Component To Local(ComponentPose=@n7)) {layerGroup=None}
  ]}
 EventGraph {nodes [
   n1 = Noise FNC!(execute, self, Value:Vector="3,2,5", Speed:Vector="1,1,1.5", Offset:Vector="0, 0, 0") -> (then, Rotator:Rotator)
   > Set RotateNoiseA(RotateNoiseA:Rotator=@n1) -> (Output_Get:Rotator)
   > n2 = Noise FNC(Value="3,6,7", Speed="1,1,1.5", Offset="0.5,0.5,-1")
   > Set RotateNoiseB(RotateNoiseB:Rotator=@n2) -> (Output_Get:Rotator)
   > n3 = Noise FNC(Value="3,10,8", Speed="1,1,1.5", Offset="1,1,-2")
   > Set RotateNoiseC(RotateNoiseC:Rotator=@n3) -> (Output_Get:Rotator)
   TryGetPawnOwner()
   Event BlueprintUpdateAnimation()
  ]}
 Noise FNC {nodes [
   n1 = Noise FNC!() -> (then, Value:Vector=~, Speed:Vector=~, Offset:Vector=~) {is=title K2Node_FunctionEntry}
   n2 = BreakVector(InVec=vector + float!(A:Vector=@n1.Offset, B:double=GetGameTimeInSeconds()) -> (ReturnValue:Vector))
   n3 = BreakVector(InVec=@n1.Speed)
   n4 = "Sin (Radians)"(A=float * float(A=@n2.Z, B=@n3.Z))
   n5 = BreakVector(InVec=@n1)
   n6 = "Sin (Radians)"(A=float * float(A=@n2, B=@n3))
   n7 = "Sin (Radians)"(A=float * float(A=@n2.Y, B=@n3.Y))
   > Return Node!(
    execute
    Rotator:Rotator=MakeRotator(Roll=float * float(A=@n6, B=@n5), Pitch=float * float(A=@n7, B=@n5.Y), Yaw=float * float(A=@n4, B=@n5.Z))
   )
  ]}
]
variables [
 RotateNoiseA {type=Rotator}
 RotateNoiseB {type=Rotator}
 RotateNoiseC {type=Rotator}
]
